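# Niagara system for a small magic sprites environment effect in the S13 Sparks maps set, going by its name and path.
# It holds one emitter, Minimal, with no listed renderers, and SystemState is the only listed module.
# No user variables, data interfaces or dynamic inputs are listed.
# The LWE suffix suggests a lightweight emitter, but no particle behavior appears here.
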NiagaraSystem /SparksMaps/Effects/Environment/S13/MagicSprites/NS_Sparks_S13_MagicSprites_LWE_Small
dependencies {usedDataInterfaces [], usedDynamicInputs [], usedModules [&SystemState]}
emitters {Minimal {rendererClasses [], renderers []}}
summary {systemName=NS_Sparks_S13_MagicSprites_LWE_Small, emitters [{emitterName=Minimal, rendererClasses []}], userVariables []}
